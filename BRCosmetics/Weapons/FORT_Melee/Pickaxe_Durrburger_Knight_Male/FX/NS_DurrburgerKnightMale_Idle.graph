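# Niagara idle effect for the Durrburger Knight pickaxe that drips sprite droplets from skeletal mesh bones or sockets.
# Drip_A pins swelling, stretching drops to points picked by spawn group, then emits a death event with world position.
# Drip_Chest_B, which has no spawn module shown, animates falling drops with downward acceleration and speed-based stretch.
# User.DripSpawnRate drives gameplay spawn rate and both emitters' alpha, while front-end modifiers override rate, delay and attach point.
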NiagaraSystem /BRCosmetics/Weapons/FORT_Melee/Pickaxe_Durrburger_Knight_Male/FX/NS_DurrburgerKnightMale_Idle
dependencies {
 usedDataInterfaces [
  &NiagaraDataInterfaceVector2DCurve
  &NiagaraDataInterfaceCurve
  &NiagaraDataInterfaceVectorCurve
 ]
 usedDynamicInputs []
 usedModules [
  &self:SystemSpawnScript.SystemScriptSource.SystemScriptGraph.NiagaraNodeAssignment_0.SetVariables_#1
  &SystemState
 ]
}
emitters {
 Drip_A {
  emitterUpdateScript [
   &EmitterState(
    Life Cycle Mode=::NewEnumerator1
    Loop Delay=&FN_FrontEndFloatModifier(
     Gameplay=&RandomRangeFloat(Minimum=0, Maximum=1)
     Front End=&RandomRangeFloat(Minimum=1, Maximum=2.5)
    )
    Loop Duration=1
    UseLoopDelay=true
   )
   &SpawnRate(
    SpawnRate=&FN_FrontEndFloatModifier(Gameplay=$User.DripSpawnRate, Front End=1)
    Use Spawn Probability=false
   )
  ]
  particleSpawnScript [
   &FN_SetSpawnGroup(Number of Spawn Groups=2)
   &InitializeParticle(
    Lifetime=&RandomRangeFloat(Minimum=1.8, Maximum=3)
    Sprite Rotation Angle=-180
    Sprite Rotation Mode=::NewEnumerator0
    Sprite Size Mode=::NewEnumerator1
    Uniform Sprite Size Max=2
    Uniform Sprite Size Min=1.5
    UsePositionOffset=false
    Write Material Random=true
   )
  ]
  particleUpdateScript [
   &ParticleState()
   &SkeletalMeshLocation(
    Bone / Socket Index=&MakeIntFromFloat(Float=&FN_FrontEndFloatModifier(Gameplay=&MakeFloatFromInt(Int=$Particles.SpawnGroup), Front End=1))
    Inherit Velocity=false
    MaskSpawnGroup=false
    Mesh Sampling Type=::NewEnumerator5
    Sampled Position Offset=<0 0 1.35>
    Skeletal Mesh=$User.SkeletalMesh
    Socket Sampling Mode=::NewEnumerator5
   )
   &SolveForcesAndVelocity(Clamp Velocity=false, Limit Acceleration=false)
   &ScaleSpriteSize(Scale Factor=&Vector2DFromCurve(
     Vector2Curve=DataInterface() {propertyValues {
       XCurve {keys [
         {interpMode=RCIM_Cubic, tangentMode=RCTM_User, time=0, value=1.15}
         {arriveTangent=-0.8621, interpMode=RCIM_Cubic, leaveTangent=-0.8621, tangentMode=RCTM_User, time=0.6, value=1.14}
         <at 0.8265 0.457>
         {arriveTangent=-0.351, interpMode=RCIM_Cubic, leaveTangent=-0.351, tangentMode=RCTM_User, time=0.87, value=0.4}
         {interpMode=RCIM_Cubic, tangentMode=RCTM_User, time=0.9988, value=1}
        ]}
       YCurve {keys [
         {interpMode=RCIM_Cubic, tangentMode=RCTM_User, time=0, value=0}
         {arriveTangent=2.8736, interpMode=RCIM_Cubic, leaveTangent=2.8736, tangentMode=RCTM_User, time=0.6, value=0.99}
         {arriveTangent=2.025, interpMode=RCIM_Cubic, leaveTangent=2.025, tangentMode=RCTM_User, time=0.87, value=2.5}
         {interpMode=RCIM_Cubic, tangentMode=RCTM_User, time=1, value=1.8}
        ]}
      }}
     ScaleCurve=<1 1>
    ))
   &GenerateDeathEvent(
    Gap Correction Amount=1.5
    Position To Send:Vector3f=/Niagara/DynamicInputs/Transforms/TransformPosition(Position=Linked() {linkedVariable {name=Particles.Position, type=NiagaraPosition}}, Source Space=::Local, Destination Space=::World)
   )
   &ScaleColor(
    Scale Alpha=&FloatFromCurve(
     FloatCurve=DataInterface() {propertyValues {Curve {keys [<at 0 0>,<at 0.1 1>,<at 0.9 1>,<at 1.0011 0.002>]}}}
     Scale Curve=$User.DripSpawnRate
    )
    ScaleA=true
    ScaleRGB=true
   )
  ]
  rendererClasses [&NiagaraSpriteRendererProperties]
  renderers [{rendererClass=&NiagaraSpriteRendererProperties}]
 }
 Drip_Chest_B {
  emitterUpdateScript [&EmitterState(Life Cycle Mode=::NewEnumerator1, Loop Duration=1, UseLoopDelay=false)]
  particleSpawnScript [
   &InitializeParticle(
    Lifetime=&RandomRangeFloat(Minimum=0.5, Maximum=2)
    Sprite Size Mode=::NewEnumerator3
    Uniform Sprite Size=3
    UsePositionOffset=false
    Write Material Random=true
   )
   &InheritVelocity(Module.Limit Speed=false, Velocity Scale=<0.5 0.5 0.5>)
  ]
  particleUpdateScript [
   &ParticleState()
   &AccelerationForce(Acceleration=&VectorFromCurve(VectorCurve=DataInterface() {propertyValues {
       XCurve {keys [<at 0 0>,<at 0.9989 0>]}
       YCurve {keys [<at 0 0>,<at 0.9989 0>]}
       ZCurve {keys [
         {arriveTangent=-15.5802, interpMode=RCIM_Cubic, leaveTangent=-8.2205, tangentMode=RCTM_User, time=0, value=0}
         {arriveTangent=-962.5264, interpMode=RCIM_Cubic, leaveTangent=-1066.8696, tangentMode=RCTM_User, time=0.3, value=-350}
        ]}
      }}))
   &SolveForcesAndVelocity(Clamp Velocity=false, Limit Acceleration=false)
   &ScaleSpriteSizeBySpeed(Max Scale Factor=<1 3>, Velocity Threshold=100)
   &ScaleSpriteSize(Scale Factor=&Vector2DFromCurve(
     Vector2Curve=DataInterface() {propertyValues {
       XCurve {keys [
         {arriveTangent=5.9728, interpMode=RCIM_Cubic, leaveTangent=5.9728, tangentMode=RCTM_User, time=0, value=0.1}
         {arriveTangent=-0.1, interpMode=RCIM_Cubic, leaveTangent=-0.1, tangentMode=RCTM_User, time=0.18, value=0.99}
         {interpMode=RCIM_Cubic, tangentMode=RCTM_User, time=1, value=0}
        ]}
       YCurve {keys [
         {arriveTangent=6.2715, interpMode=RCIM_Cubic, leaveTangent=6.2715, tangentMode=RCTM_User, time=0, value=0.1}
         {arriveTangent=-0.1001, interpMode=RCIM_Cubic, leaveTangent=-0.1001, tangentMode=RCTM_User, time=0.18, value=0.99}
         {interpMode=RCIM_Cubic, tangentMode=RCTM_User, time=0.9989, value=0}
        ]}
      }}
     ScaleCurve=<1 1>
    ))
   &ScaleColor(Scale Alpha=$User.DripSpawnRate, ScaleA=true, ScaleRGB=true)
  ]
  rendererClasses [&NiagaraSpriteRendererProperties]
  renderers [{rendererClass=&NiagaraSpriteRendererProperties}]
 }
}
summary {
 systemName=NS_DurrburgerKnightMale_Idle
 emitters [
  {emitterName=Drip_A, rendererClasses [&NiagaraSpriteRendererProperties]}
  {emitterName=Drip_Chest_B, rendererClasses [&NiagaraSpriteRendererProperties]}
 ]
 userVariables [
  User.bIsFrontEnd {type=NiagaraBool}
  User.bIsFrontEndPreview {type=NiagaraBool}
  User.DripSpawnRate {type=NiagaraFloat, defaultValue=1, description=""}
  User.JumpBurst {type=NiagaraFloat, defaultValue=0, description=""}
  User.SkeletalMesh {
   type=NiagaraDataInterfaceSkeletalMesh
   defaultValue {
    struct=NiagaraExt_VariableValue_DataInterface
    value {dataInterface=&self:NiagaraDataInterfaceSkeletalMesh_0}
   }
  }
 ]
}
